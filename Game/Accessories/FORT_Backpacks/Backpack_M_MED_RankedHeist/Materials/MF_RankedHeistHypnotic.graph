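# Material function for the Ranked Heist backpack that overlays animated "hypnotic" contour lines on incoming material attributes.
# Two samples of an SDF texture on UV1, each with its own offset, rotation and tiling, scroll their distance over time into repeating lines, with green-channel thickness and screen-derivative anti-aliasing.
# Both line sets tint base color with separate colors and drive metallic, specular and roughness changes.
# A channel-masked Alpha input blends the result over the original.
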MaterialFunction /Game/Accessories/FORT_Backpacks/Backpack_M_MED_RankedHeist/Materials/MF_RankedHeistHypnotic
graph {nodes [
  Constant() {r=0.008}
  Constant() {r=0.008}
  n1 = SetMaterialAttributes!(
   MaterialAttributes=NamedRerouteUsage()
   BaseColor=NamedRerouteUsage()
   Metallic=NamedRerouteUsage()
   Roughness=NamedRerouteUsage()
   Specular=NamedRerouteUsage()
  ) {attributeSetTypes [
    #1
    #2
    #3
    #4
   ]}
  FunctionOutput(None=BlendMaterialAttributes(A=NamedRerouteUsage(), B=@n1, Alpha=NamedRerouteUsage()))
  n2 = TextureCoordinate() {coordinateIndex=1}
  n3 = MaterialFunctionCall(
   X=ScalarParameter() {defaultValue=0.9, group=RankedHeist, parameterName=HypnoticTileX}
   Y=ScalarParameter() {defaultValue=0.9, group=RankedHeist, parameterName=HypnoticTileY}
  ) {materialFunction=&MakeFloat2}
  Multiply(A=@n2, B=@n3.Result)
  n4 = ScalarParameter() {defaultValue=-0.115, group=RankedHeist, parameterName=HypnoticOffsetX}
  n5 = ScalarParameter() {defaultValue=0.065, group=RankedHeist, parameterName=HypnoticOffsetY}
  n6 = TextureSample(UVs=MaterialFunctionCall!(
    UV=@n2
    Offset=MaterialFunctionCall(X=@n4, Y=@n5) {materialFunction=&MakeFloat2}.Result
    Rotation=ScalarParameter() {defaultValue=-0.06, group=RankedHeist, parameterName=HypnoticRotation}
    Scale=@n3.Result
    Pivot
   ) {materialFunction=/Game/Global/MaterialFunctions/MF_UV_Transform}.Result) {
   samplerType=SAMPLERTYPE_LinearColor
   texture=/Game/Accessories/FORT_Backpacks/Backpack_M_MED_RankedHeist/Textures/T_Backpack_Ranked_Heist_SDF
  }
  NamedRerouteDeclaration(None=@n6.R) {name=SDF, nodeColor=<rgba 0 1 0.8125 1>}
  NamedRerouteDeclaration(None=@n6.G) {name=Thickness, nodeColor=<rgba 0.7344 1 0 1>}
  n7 = ChannelMaskParameter(None=FunctionInput() {bUsePreviewValueAsDefault=true, inputName=Alpha, sortPriority=1}) {group=RankedHeist, maskChannel=Green, parameterName=RH_HypnoticMaskChannel, defaultValue=<rgba 0 1 0 0>}
  NamedRerouteDeclaration(None=@n7.R) {name=IN_Alpha, nodeColor=<rgba 0 0.6953 1 1>}
  n8 = GetMaterialAttributes(None=NamedRerouteDeclaration(None=FunctionInput() {inputName=Material, inputType=FunctionInput_MaterialAttributes}) {name=IN_Material, nodeColor=<rgba 0 1 0.6484 1>}) {attributeGetTypes [
    #1
    #3
    #2
    #4
   ]}
  NamedRerouteDeclaration(None=@n8.Roughness) {name=IN_Roughness, nodeColor=<rgba 0.3828 0 1 1>}
  NamedRerouteDeclaration(None=LinearInterpolate(A=NamedRerouteUsage(), B=ScalarParameter() {defaultValue=0.4786, parameterName=HypnoticLinesRoughness}, Alpha=NamedRerouteUsage())) {name=OUT_Roughness, nodeColor=<rgba 0.8984 1 0 1>}
  n9 = OneMinus(None=NamedRerouteUsage())
  n10 = OneMinus(None=NamedRerouteUsage())
  NamedRerouteDeclaration(None=Saturate(None=Add(A=@n9, B=@n10))) {name=LineMask, nodeColor=<rgba 0 1 0.2734 1>}
  NamedRerouteDeclaration(None=@n8.Metallic) {name=IN_Metal, nodeColor=<rgba 0 1 0.0625 1>}
  NamedRerouteDeclaration(None=@n8.Specular) {name=IN_Spec, nodeColor=<rgba 0 1 0.0625 1>}
  NamedRerouteDeclaration(None=LinearInterpolate(A=NamedRerouteUsage(), B=ScalarParameter() {defaultValue=0.3, parameterName=HypnoticLinesMetal}, Alpha=NamedRerouteUsage())) {name=OUT_Metal, nodeColor=<rgba 0.8984 1 0 1>}
  NamedRerouteDeclaration(None=LinearInterpolate(A=NamedRerouteUsage(), B=ScalarParameter() {defaultValue=0.4, parameterName=HypnoticLinesSpec}, Alpha=NamedRerouteUsage())) {name=OUT_Spec, nodeColor=<rgba 0.8984 1 0 1>}
  n11 = ScalarParameter() {defaultValue=0.125, group=RankedHeist, parameterName=HypnoticOffsetX2}
  n12 = ScalarParameter() {defaultValue=-0.05, group=RankedHeist, parameterName=HypnoticOffsetY2}
  n13 = ScalarParameter() {defaultValue=0.9, group=RankedHeist, parameterName=HypnoticTileX2}
  n14 = ScalarParameter() {defaultValue=0.9, group=RankedHeist, parameterName=HypnoticTileY2}
  n15 = TextureSample(UVs=MaterialFunctionCall(
    UV=TextureCoordinate() {coordinateIndex=1}
    Offset=MaterialFunctionCall(X=@n11, Y=@n12) {materialFunction=&MakeFloat2}.Result
    Rotation=ScalarParameter() {defaultValue=0.03, group=RankedHeist, parameterName=HypnoticRotation2}
    Scale=MaterialFunctionCall(X=@n13, Y=@n14) {materialFunction=&MakeFloat2}.Result
   ) {materialFunction=&MF_UV_Transform}.Result) {
   samplerType=SAMPLERTYPE_LinearColor
   texture=&T_Backpack_Ranked_Heist_SDF
  }
  NamedRerouteDeclaration(None=@n15.R) {name=SDF_1, nodeColor=<rgba 0 1 0.8125 1>}
  n16 = Clamp(None=MaterialFunctionCall(
    Input=NamedRerouteUsage()
    Input Low=Constant()
    Input High=Constant() {r=1}
    Target Low=Constant() {r=0.0065}
    Target High=Constant() {r=0.045}
   ) {materialFunction=&RemapValueRange}.Result)
  n17 = Time()
  n18 = ScalarParameter() {defaultValue=-0.05, group=RankedHeist, parameterName=Speed}
  n19 = NamedRerouteDeclaration(None=Add(A=Multiply(A=@n17, B=@n18), B=NamedRerouteUsage())) {name=SDFModified, nodeColor=<rgba 0 0.6953 1 1>}
  n20 = DDX(Value=@n19)
  n21 = DDY(Value=@n19)
  n22 = Multiply(A=Add(A=Abs(None=@n20), B=Abs(None=@n21))) {constB=0.5}
  n23 = ScalarParameter() {defaultValue=0.1, group=RankedHeist, parameterName=LineDistance}
  n24 = Add(A=Divide(A=@n19, B=@n23)) {constB=0.5}
  n25 = Abs(None=Subtract(A=Frac(None=@n24)) {constB=0.5})
  NamedRerouteDeclaration(None=SmoothStep(Min=Subtract(A=@n16, B=@n22), Max=Add(A=@n16, B=@n22), Value=Multiply(A=@n25, B=@n23))) {name=Lines, nodeColor=<rgba 1 0 0.8906 1>}
  NamedRerouteDeclaration(None=@n15.G) {name=Thickness_1, nodeColor=<rgba 0.7344 1 0 1>}
  n26 = Clamp(None=MaterialFunctionCall(
    Input=NamedRerouteUsage()
    Input Low=Constant()
    Input High=Constant() {r=1}
    Target Low=Constant() {r=0.007}
    Target High=Constant() {r=0.04}
   ) {materialFunction=&RemapValueRange}.Result)
  n27 = Time()
  n28 = ScalarParameter() {defaultValue=-0.05, group=RankedHeist, parameterName=Speed2}
  n29 = NamedRerouteDeclaration(None=Add(A=Multiply(A=@n27, B=@n28), B=NamedRerouteUsage())) {name=SDFModified_1, nodeColor=<rgba 0 0.6953 1 1>}
  n30 = DDX(Value=@n29)
  n31 = DDY(Value=@n29)
  n32 = Multiply(A=Add(A=Abs(None=@n30), B=Abs(None=@n31))) {constB=0.5}
  n33 = ScalarParameter() {defaultValue=0.1, group=RankedHeist, parameterName=LineDistance}
  n34 = Add(A=Divide(A=@n29, B=@n33)) {constB=0.5}
  n35 = Abs(None=Subtract(A=Frac(None=@n34)) {constB=0.5})
  NamedRerouteDeclaration(None=SmoothStep(Min=Subtract(A=@n26, B=@n32), Max=Add(A=@n26, B=@n32), Value=Multiply(A=@n35, B=@n33))) {name=Lines_1, nodeColor=<rgba 1 0 0.8906 1>}
  NamedRerouteDeclaration(None=@n8.BaseColor) {name=IN_Color, nodeColor=<rgba 0 0.4141 1 1>}
  n36 = NamedRerouteUsage()
  n37 = VectorParameter() {group=RankedHeist, parameterName=Color_01, defaultValue=<rgba 0 0 0 1>}
  n38 = OneMinus(None=NamedRerouteUsage())
  n39 = NamedRerouteUsage()
  NamedRerouteDeclaration(None=LinearInterpolate(
    A=LinearInterpolate(A=@n36, B=@n37.RGB, Alpha=@n38)
    B=VectorParameter() {group=RankedHeist, parameterName=Color_02, defaultValue=<rgba 0.0922 0.3542 0.2971 1>}.RGB
    Alpha=OneMinus(None=@n39)
   )) {name=OUT_Color, nodeColor=<rgba 0 1 0.25 1>}
  NamedRerouteUsage()
  NamedRerouteUsage()
  NamedRerouteUsage()
 ]}
